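; Conversation side effect that requests a time-of-day change when triggered from an NPC dialogue service in Fortnite.
; The client side-effect function gets the participant player from the conversation context and finds the event router on the game state.
; It broadcasts an OnTimeOfDayChangeRequested event carrying the Athena playlist context tags, leaving the actual change to whatever listens for that event.
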(fn ClientCauseSideEffect
  (params (Context :type Conversation%20Context%20Structure%20%28by%20ref%29))
  (bind _returnvalue (Class|FortConversationContextLibrary|GetParticipantActorPlayer Context))
  (bind _returnvalue_2 (GameplayEventRouter|GetEventRouter "/Script/FortniteGame.FortGameState" _returnvalue))
  (GameplayEventRouter|BroadcastEvent _returnvalue_2 (Utilities|Struct|MakeOnTimeOfDayChangeRequested (Class|FortGameStateAthena|GetAthenaPlaylistContextTags (Fortnite|Athena|GameState|GetGameStateAthenafromContext _returnvalue))) 0)
  (return))
