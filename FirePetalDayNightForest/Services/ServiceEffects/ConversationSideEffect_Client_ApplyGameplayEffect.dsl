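; Conversation side effect that applies a configurable gameplay effect to a conversation participant on the client.
; A participant enum variable selects either the player or NPC actor from the conversation context.
; A helper resolves that actor's Ability System Component and returns it if valid, otherwise none.
; The client side-effect event then applies the effect class to that component as a self-applied effect at a fixed level.
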(fn ClientCauseSideEffect
  (params (Context :type Conversation%20Context%20Structure%20%28by%20ref%29))
  (bind _participantabilitysystem (CallFunction|GetTargetAbilityComponent Context))
  (GameplayEffects|ApplyGameplayEffectToSelf :self _participantabilitysystem :GameplayEffectClass (Variables|Default|GetgameplayEffect) :Level 1.0)
  (return))

(fn getTargetAbilityComponent
  (params (context :type Conversation%20Context%20Structure%20%28by%20ref%29))
  (switch Utilities|FlowControl|Switch|SwitchonEConversationParticipant (Variables|Default|GetconversationParticipant)
    (:NewEnumerator0
      (bind _returnvalue (Class|FortConversationContextLibrary|GetParticipantActorPlayer context))
      (Utilities|IsValid (Ability|GetAbilitySystemComponent _returnvalue)
        (:Is%20Valid
          (bind _output_get (LocalVariables|Set_actorAbilitySystem (Ability|GetAbilitySystemComponent _returnvalue)))
          (return _output_get))
        (:Is%20Not%20Valid
          _)))
    (:NewEnumerator1
      (bind _returnvalue_2 (Class|FortConversationContextLibrary|GetParticipantActorNPC context))
      (Utilities|IsValid (Ability|GetAbilitySystemComponent _returnvalue_2)
        (:Is%20Valid
          (bind _output_get_2 (LocalVariables|Set_actorAbilitySystem (Ability|GetAbilitySystemComponent _returnvalue_2)))
          (return _output_get_2))
        (:Is%20Not%20Valid
          _))))
  (return 0))
