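; A conversation side effect that fires a one-shot (burst) gameplay cue on the client during an NPC dialogue.
; GetTargetActor switches on the configured conversation participant to return either the player or the NPC actor from the conversation context.
; ClientCauseSideEffect executes the configured gameplay cue tag on that actor, passing a simple Fort effect context built from it.
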(fn ClientCauseSideEffect
  (params (Context :type Conversation%20Context%20Structure%20%28by%20ref%29))
  (bind _participantactor (CallFunction|GetTargetActor Context))
  (GameplayCue|ExecuteGameplayCue _participantactor (Variables|Default|GetgameplayCueTag) (Class|AthenaFunctionLibrary|MakeSimpleFortEffectContext _participantactor))
  (return))

(fn getTargetActor
  (params (context :type Conversation%20Context%20Structure%20%28by%20ref%29))
  (switch Utilities|FlowControl|Switch|SwitchonEConversationParticipant (Variables|Default|GetconversationParticipant)
    (:NewEnumerator0
      (bind _returnvalue (Class|FortConversationContextLibrary|GetParticipantActorPlayer context))
      (return _returnvalue))
    (:NewEnumerator1
      (bind _returnvalue_2 (Class|FortConversationContextLibrary|GetParticipantActorNPC context))
      (return _returnvalue_2))))
