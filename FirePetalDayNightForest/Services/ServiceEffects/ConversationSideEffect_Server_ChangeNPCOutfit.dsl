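; Server-side conversation side effect that changes the outfit of the NPC taking part in a conversation.
; It gets the NPC from the conversation context and proceeds only if its conversation component's entry tag matches the configured tag, with optional exact matching.
; It then generates a GUID from the new outfit asset and calls the Fortnite pawn's cosmetic swap with the configured variant data.
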(fn findConversationTag
  (params (NPCActor :type Actor%20Object%20Reference))
  (bind _returnvalue (Conversation|FindConversationComponent NPCActor))
  (Utilities|IsValid _returnvalue
    (:Is%20Valid
      (bind _asnpcconversation_component (Utilities|Casting|CastToNPCConversationComponent _returnvalue)
        (:then
          (return (GameplayTags|MatchesTag (Variables|/Script/FortniteGame.FortNonPlayerConversationParticipantComponent|GetConversationEntryTag _asnpcconversation_component) (Variables|ConversationTag|GetconversationEntryTag) (Variables|ConversationTag|GetrequireExactMatch))))
        (:CastFailed
          _)))
    (:Is%20Not%20Valid
      _))
  (return false))

(fn ServerCauseSideEffect
  (params (Context :type Conversation%20Context%20Structure%20%28by%20ref%29))
  (bind _returnvalue (Class|FortConversationContextLibrary|GetParticipantActorNPC Context))
  (bind _foundtag (CallFunction|FindConversationTag _returnvalue))
  (if _foundtag
    (CallFunction|SwapOutfitVariant _returnvalue)
    (else
      _))
  (return))

(fn swapOutfitVariant
  (params (NPCActor :type Actor%20Object%20Reference))
  (bind _returnvalue (GUID|GenerateGuidfromObject (Variables|Outfit|GetnewOutfit)))
  (bind _output_get (LocalVariables|Set_outfitGuid _returnvalue))
  (bind _asfort_player_pawn (Utilities|Casting|CastToFortPlayerPawn NPCActor)
    (:then
      (Class|FortPlayerPawn|SwapPawnCosmeticItemandVariant _asfort_player_pawn _output_get (Variables|Outfit|GetoutfitVariantData) false "Active" false))
    (:CastFailed
      _))
  (return))
