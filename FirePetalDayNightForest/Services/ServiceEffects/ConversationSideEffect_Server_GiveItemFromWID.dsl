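; Server-side conversation side effect that gives the player in the conversation one copy of a configured item.
; ServerCauseSideEffect resolves the player participant from the conversation context as a FortPlayerController.
; It then casts the itemDefinition variable to a FortItemDefinition and adds that item through the controller's inventory owner interface.
; A missing player or failed cast ends the effect silently.
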(fn AddItemToInstigatorInventory
  (params (InventoryOwner :type Fort%20Inventory%20Owner%20Interface%20Interface) (ItemDefinition :type Fort%20Item%20Definition%20Object%20Reference))
  (InventoryManagement|AddItem InventoryOwner ItemDefinition 1)
  (return))

(fn GetItemReference
  (bind _asfort_item_definition (Utilities|Casting|CastToFortItemDefinition (Variables|Default|GetitemDefinition))
    (:then
      (LocalVariables|Set_validItem true))
    (:CastFailed
      (LocalVariables|Set_validItem false)))
  (return (LocalVariables|Get_validItem) _asfort_item_definition))

(fn GetPlayerReference
  (params (Context :type Conversation%20Context%20Structure%20%28by%20ref%29))
  (bind _returnvalue (Class|FortConversationContextLibrary|GetParticipantActorPlayer Context))
  (Utilities|IsValid _returnvalue
    (:Is%20Valid
      (bind _asfort_player_controller (Utilities|Casting|CastToFortPlayerController _returnvalue)
        (:then
          (return _asfort_player_controller true))
        (:CastFailed
          _)))
    (:Is%20Not%20Valid
      _))
  (return 0 false))

(fn ServerCauseSideEffect
  (params (Context :type Conversation%20Context%20Structure%20%28by%20ref%29))
  (bind (_playercontroller _success) (CallFunction|GetPlayerReference Context))
  (if _success
    (bind (_validitem _itemdefinition) (CallFunction|GetItemReference))
    (if _validitem
      (CallFunction|AddItemtoInstigatorInventory _playercontroller _itemdefinition)
      (else
        _))
    (else
      _))
  (return))
